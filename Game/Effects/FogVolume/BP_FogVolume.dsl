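; Fog volume actor that renders local fog on a cube using a dynamic material, hidden on mobile.
; Construction and BeginPlay run UpdateParticle, which creates a missing material instance from FogType and RadialFalloff, named by actor name and position.
; Only upon creation, UpdateMaterial pushes noise, color, time-of-day density, falloff and cube bounds into it.
; Fog type 1 uses the V2 material and adds Fortnitemares parameters.
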(event EventBeginPlay
  (bind _fogcube (Variables|Default|GetFogCube))
  (CallFunction|UpdateParticle)
  (if (TouchPlatform|IsMobileGame)
    (Rendering|SetVisibility _fogcube false false)
    (else
      (Rendering|SetVisibility _fogcube true false))))

(fn SetScalarOnFogMaterial
  (params (Value :type Float%20%28double-precision%29) (Name :type Name))
  (Rendering|Material|SetScalarParameterValue (Variables|Default|GetFogMaterial) Name Value :!class /Script/Engine.MaterialInstanceDynamic))

(fn SetVectorOnFogMaterial
  (params (Value :type Linear%20Color%20Structure) (Name :type Name))
  (Rendering|Material|SetVectorParameterValue (Variables|Default|GetFogMaterial) Name Value :!class /Script/Engine.MaterialInstanceDynamic))

(fn UpdateMaterial
  (bind _fogcube (Variables|Default|GetFogCube))
  (bind (_origin _boxextent) (Collision|GetComponentBounds _fogcube))
  (bind _returnvalue (- _origin _boxextent))
  (bind _returnvalue_2 (Math|Conversions|ToLinearColor(Vector) _returnvalue))
  (bind _returnvalue_3 (+ _origin _boxextent))
  (bind _returnvalue_4 (Math|Conversions|ToLinearColor(Vector) _returnvalue_3))
  (switch int (Variables|FogTypeSelec|GetFogType)
    (:Default
      _)
    (:0
      _)
    (:1
      (CallFunction|SetScalaronFogMaterial (Variables|Fortnitemares|GetFogGlobalEmissive) "FogGlobalEmissive")
      (CallFunction|SetScalaronFogMaterial (Variables|Fortnitemares|GetGlobalExtinctionMultiplier) "GlobalExtinctionMultiplier")
      (CallFunction|SetScalaronFogMaterial (Variables|Fortnitemares|GetFogEmissiveStrength) "FogEmissiveStregnth")
      (CallFunction|SetScalaronFogMaterial (Variables|Fortnitemares|GetFogUnderGlowStrength) "UnderGlowStrength")
      (CallFunction|SetScalaronFogMaterial (Variables|Fortnitemares|GetFogShadowDensity) "ShadowDensity")
      (CallFunction|SetScalaronFogMaterial (Variables|Fortnitemares|GetFogShadowOffset) "OffsetAmt")
      (CallFunction|SetScalaronFogMaterial (Variables|Fortnitemares|GetGlobalNoiseScale) "GlobalNoiseScale")
      (CallFunction|SetScalaronFogMaterial (Variables|Fortnitemares|GetCurlNoiseScale) "CurlNoiseScale")
      (CallFunction|SetScalaronFogMaterial (Variables|Fortnitemares|GetNoiseScale) "NoiseScale")
      (CallFunction|SetScalaronFogMaterial (Variables|Fortnitemares|GetDistortionStrength) "DistortionStrength")
      (CallFunction|SetScalaronFogMaterial (Variables|Fortnitemares|GetNoiseStrength) "Noisetrength")
      (CallFunction|SetScalaronFogMaterial (Variables|Fortnitemares|GetFogRadiusMultiplier) "FogRadiusMultiplier")
      (CallFunction|SetScalaronFogMaterial (Variables|Fortnitemares|GetFogRadiusHardness) "Hardness")
      (CallFunction|SetScalaronFogMaterial (Variables|Fortnitemares|GetFog_MorningMaxDensity) "Morning_FogMax")
      (CallFunction|SetScalaronFogMaterial (Variables|Fortnitemares|GetFog_DayMaxDensity) "Day_FogMax")
      (CallFunction|SetScalaronFogMaterial (Variables|Fortnitemares|GetFog_EveningMaxDensity) "Evening_FogMax")
      (CallFunction|SetScalaronFogMaterial (Variables|Fortnitemares|GetFog_NightMaxDensity) "Night_FogMax")
      (CallFunction|SetScalaronFogMaterial (Variables|Fortnitemares|GetFog_MorningMinDensity) "Morning_FogMin")
      (CallFunction|SetScalaronFogMaterial (Variables|Fortnitemares|GetFog_DayMinDensity) "Day_FogMin")
      (CallFunction|SetScalaronFogMaterial (Variables|Fortnitemares|GetFog_EveningMinDensity) "Evening_FogMin")
      (CallFunction|SetScalaronFogMaterial (Variables|Fortnitemares|GetFog_NightMinDensity) "Night_FogMin")
      (CallFunction|SetScalaronFogMaterial (Variables|Fortnitemares|GetFog_MacroScale) "FogMacroScale")
      (CallFunction|SetScalaronFogMaterial (Variables|Fortnitemares|GetFog_MacroMaxDensity) "Macro_MAxDensityMultiplier")
      (CallFunction|SetScalaronFogMaterial (Variables|Fortnitemares|GetFog_MacroMinDensity) "Macro_MinDensityMultiplier")
      (CallFunction|SetScalaronFogMaterial (Variables|Fortnitemares|GetFog_MacroMinDensity) "Macro_MinDensityMultiplier")
      (CallFunction|SetScalaronFogMaterial (Variables|Fortnitemares|GetFog_DistanceBlendMultiplier) "DistanceBlendMultiplier")
      (CallFunction|SetScalaronFogMaterial (Variables|Fortnitemares|GetFog_MacroNoiseSpeedMultiplier) "MacroPanSpeedMultiplier")
      (CallFunction|SetVectoronFogMaterial (Math|Conversions|ToLinearColor(Vector) (Variables|Fortnitemares|GetFog_CurlNoisePanSpeedAndDirection)) "CurlNoisePanSpeedAndDirection")
      (CallFunction|SetVectoronFogMaterial (Math|Conversions|ToLinearColor(Vector) (Variables|Fortnitemares|GetFog_NoisePanSpeedAndDirection)) "FogNoisePanSpeed")
      (CallFunction|SetVectoronFogMaterial _returnvalue_2 "BBoxMin")
      (CallFunction|SetVectoronFogMaterial _returnvalue_4 "BBoxMax")
      (CallFunction|SetVectoronFogMaterial (Math|Conversions|ToLinearColor(Vector) (Variables|Default|GetAxisFalloffExponent)) "AxisFallfoffExponent")
      (CallFunction|SetVectoronFogMaterial (Math|Conversions|ToLinearColor(Vector) (Variables|Default|GetAxisShift)) "AxisShift")))
  (CallFunction|SetScalaronFogMaterial (Variables|Density|GetDensity_Day) "Density_Day")
  (CallFunction|SetScalaronFogMaterial (Variables|Density|GetDensity_Morning) "Density_Morning")
  (CallFunction|SetScalaronFogMaterial (Variables|Density|GetDensity_Evening) "Density_Evening")
  (CallFunction|SetScalaronFogMaterial (Variables|Density|GetDensity_Night) "Density_Night")
  (CallFunction|SetScalaronFogMaterial (Variables|Default|GetNoiseSpeed) "NoiseSpeed")
  (CallFunction|SetScalaronFogMaterial (Variables|Default|GetNoiseMin) "NoiseMin")
  (CallFunction|SetScalaronFogMaterial (Variables|Default|GetNoiseMax) "NoiseMax")
  (CallFunction|SetScalaronFogMaterial (Variables|Default|GetNoNoiseOpacityValue) "NoNoiseOpacityValue")
  (CallFunction|SetScalaronFogMaterial (Variables|Default|GetNoiseSize) "NoiseSize")
  (CallFunction|SetScalaronFogMaterial (Variables|Default|GetRadiusScale) "RadiusScale")
  (CallFunction|SetVectoronFogMaterial (Variables|Default|GetInnerScatter) "Albedo")
  (CallFunction|SetVectoronFogMaterial (Variables|Default|GetEmissive) "Emissive")
  (Rendering|Material|SetTextureParameterValue (Variables|Default|GetFogMaterial) "NoiseTexture" (Variables|Default|GetNoiseTexture))
  (CallFunction|SetVectoronFogMaterial _returnvalue_2 "BBoxMin")
  (CallFunction|SetVectoronFogMaterial _returnvalue_4 "BBoxMax")
  (CallFunction|SetVectoronFogMaterial (Math|Conversions|ToLinearColor(Vector) (Variables|Default|GetAxisFalloffExponent)) "AxisFallfoffExponent")
  (CallFunction|SetVectoronFogMaterial (Math|Conversions|ToLinearColor(Vector) (Variables|Default|GetAxisShift)) "AxisShift"))

(fn updateParticle
  (bind _fogcube (Variables|Default|GetFogCube))
  (Utilities|FlowControl|Sequence
    (:then_0
      (Utilities|IsValid (Variables|Default|GetFogMaterial)
        (:Is%20Not%20Valid
          (bind _returnvalue (Rendering|Material|CreateDynamicMaterialInstance (Utilities|Select :Index (Variables|FogTypeSelec|GetFogType) :Option%200 (select (Variables|Default|GetRadialFalloff) "/Game/Effects/FogVolume/M_FogVolume_Radial.M_FogVolume_Radial" "/Game/Effects/FogVolume/M_FogVolume.M_FogVolume") :Option%201 "/Game/Effects/FogVolume/M_FogVolume_V2.M_FogVolume_V2") (Utilities|String|StringToName (Utilities|String|Append (Utilities|GetDisplayName self) (Utilities|String|ToString(Integer) (Math|Float|Truncate (/ (Math|Float|float+float (.x (Transformation|GetActorLocation self)) (.y (Transformation|GetActorLocation self)) (.z (Transformation|GetActorLocation self))) 3.0))))) "None" :!class /Script/Engine.KismetMaterialLibrary))
          (Variables|Default|SetFogMaterial _returnvalue)
          (Rendering|Material|SetMaterial (Variables|Default|GetFogCube) 0 (Variables|Default|GetFogMaterial))
          (CallFunction|UpdateMaterial))))
    (:then_1
      (Utilities|IsValid (Variables|Default|GetFogMaterial)
        (:Is%20Valid
          (Development|SetHiddeninGame (Variables|Default|GetFogCube) false false))))
    (:then_2
      (if (TouchPlatform|IsMobileGame)
        (Rendering|SetVisibility _fogcube false false)
        (else
          (Rendering|SetVisibility _fogcube true false))))))

(fn ConstructionScript
  (CallFunction|UpdateParticle))
