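# Additive volume-domain material for fog volumes, scaling an emissive color by noise, density and shape masks.
# Two scrolling world-space noise projections, offset along their third axis, combine by minimum; Low and Medium quality use a constant instead.
# Density follows time of day with optional radial falloff.
# A shiftable bounding-box falloff multiplies in X twice and Z once, leaving Y unused.
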Material /Game/Effects/FogVolume/M_FogVolume
graph {
 properties {
  bUsedWithStaticLighting=true
  blendMode=BLEND_Additive
  materialDomain=MD_Volume
  pixelDepthOffsetMode=PDOM_Legacy
  usedShadingModels=Default Lit
 }
 nodes [
  n1 = Divide(A=WorldPosition().XYZ, B=ScalarParameter() {defaultValue=5000, parameterName=NoiseSize})
  n2 = ComponentMask(None=@n1) {g=true, r=true}
  n3 = Multiply(A=Time(), B=ScalarParameter() {defaultValue=0.01, parameterName=NoiseSpeed})
  n4 = ComponentMask(None=@n1) {b=true}
  n5 = TextureSampleParameter2D!(UVs=Add(A=Add(A=@n2, B=@n3), B=Divide(A=@n4) {constB=4}), Level, Apply View MipBias) {constMipValue=0, mipValueMode=TMVM_MipLevel, parameterName=NoiseTexture, texture=&/T_TileNoise}
  n6 = ComponentMask(None=@n1) {b=true, g=true}
  n7 = Multiply(A=@n3) {constB=0.5}
  n8 = ComponentMask(None=@n1) {r=true}
  n9 = TextureSampleParameter2D(UVs=Add(A=Add(A=@n6, B=@n7), B=Divide(A=@n8) {constB=4})) {constMipValue=0, mipValueMode=TMVM_MipLevel, parameterName=NoiseTexture, texture=&T_TileNoise}
  n10 = Clamp(None=LinearInterpolate(A=ScalarParameter() {parameterName=NoiseMin}, B=ScalarParameter() {defaultValue=1, parameterName=NoiseMax}, Alpha=Min(A=@n5.R, B=@n9.R)))
  n11 = ScalarParameter() {defaultValue=0.5, parameterName=NoNoiseOpacityValue}
  n12 = QualitySwitch(Default=@n10, Low=@n11, High=@n10, Medium=@n11)
  n13 = WorldPosition()
  n14 = ObjectPositionWS()
  n15 = ObjectRadius()
  n16 = ScalarParameter() {defaultValue=1, parameterName=RadiusScale}
  n17 = OneMinus(None=Divide(A=Distance(A=@n13.XYZ, B=@n14), B=Multiply(A=@n15, B=@n16) {constB=0.9}))
  n18 = Constant() {r=1}
  n19 = StaticBoolParameter() {parameterName=UseRadialFalloff}
  n20 = ScalarParameter() {defaultValue=0.1, parameterName=Density_Morning}
  n21 = ScalarParameter() {defaultValue=0.1, parameterName=Density_Day}
  n22 = ScalarParameter() {defaultValue=0.1, parameterName=Density_Evening}
  n23 = ScalarParameter() {defaultValue=0.1, parameterName=Density_Night}
  n24 = Clamp(None=Multiply(
    A=StaticSwitch(True=@n17, False=@n18, Value=@n19)
    B=MaterialFunctionCall(Value_Morning=@n20, Value_Day=@n21, Value_Evening=@n22, Value_Night=@n23) {materialFunction=&Scalar_Time_of_Day}.Result
   ))
  n25 = VectorParameter() {parameterName=BBoxMin}
  n26 = Subtract(A=WorldPosition().XYZ, B=@n25.RGB)
  n27 = Subtract(A=VectorParameter() {parameterName=BBoxMax}.RGB, B=@n25.RGB)
  n28 = Power(Base=Clamp(None=Divide(A=@n26, B=@n27)), Exp=VectorParameter() {parameterName=AxisShift, defaultValue=<rgba 1 1 1 0>}.RGB)
  n29 = Abs(None=Multiply(A=Subtract(A=@n28) {constB=0.5}) {constB=2})
  n30 = MaterialFunctionCall(Float3=Power(Base=OneMinus(None=@n29), Exp=VectorParameter() {parameterName=AxisFallfoffExponent, defaultValue=<rgba 1 1 1 1>}.RGB)) {materialFunction=&BreakOutFloat3Components}
  n31 = Multiply(A=@n30.R, B=@n30.R)
  Multiply(
   A=VectorParameter() {parameterName=Emissive, defaultValue=<rgba 0.035 0.0343 0.0303 1>}.RGB
   B=Multiply(A=Multiply(A=@n12, B=@n24), B=Multiply(A=@n31, B=@n30.B))
  )
  VectorParameter() {parameterName=Albedo, defaultValue=<rgba 1 0 0.0359 1>}
 ]
}
