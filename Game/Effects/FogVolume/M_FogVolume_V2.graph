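# Volume-domain material for animated volumetric fog, with density set per time of day.
# Curl-distorted, panning Voronoi noise runs at detail and macro scales, blended by camera distance.
# A second sample offset along the sun/moon vector yields directional shading, mapped through exponential density to a color curve with underglow.
# Radial and bounding-box falloffs and a height curve, eroded by noise, blend minimum and maximum density, which scales the glow.
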Material /Game/Effects/FogVolume/M_FogVolume_V2
graph {
 properties {blendMode=BLEND_Additive, materialDomain=MD_Volume, pixelDepthOffsetMode=PDOM_Legacy, usedShadingModels=Default Lit}
 nodes [
  n1 = CollectionParameter() {
   collection=&FortniteMaterialParameters
   parameterName=SunAndMoonAtmosphereDirectionalVector
  }
  n2 = Normalize(VectorInput=ComponentMask(None=@n1) {b=true, g=true, r=true})
  n3 = WorldPosition()
  n4 = ScalarParameter() {defaultValue=1, parameterName=GlobalNoiseScale}
  n5 = Multiply(A=ScalarParameter() {defaultValue=1024, parameterName=NoiseScale}, B=@n4)
  n6 = WorldPosition()
  n7 = Multiply(A=ScalarParameter() {defaultValue=2048, parameterName=CurlNoiseScale}, B=@n4)
  n8 = Time()
  n9 = VectorParameter() {parameterName=CurlNoisePanSpeedAndDirection, defaultValue=<rgba -0.02 -0.02 -0.02 1>}
  n10 = TextureSample(UVs=Add(A=Divide(A=@n6.XYZ, B=@n7), B=Multiply(A=@n8, B=@n9.RGB))) {samplerType=SAMPLERTYPE_LinearColor, texture=&/VT_CurlNoise_LOW}
  n11 = ScalarParameter() {defaultValue=0.05, parameterName=DistortionStrength}
  n12 = Add(
   A=Add(A=Divide(A=@n3.XYZ, B=@n5), B=Multiply(A=@n10.RGB, B=@n11))
   B=Multiply(A=Time(), B=VectorParameter() {parameterName=FogNoisePanSpeed, defaultValue=<rgba 0.01 0.015 0.015 1>}.RGB)
  )
  n13 = ScalarParameter() {defaultValue=0.15, parameterName=OffsetAmt}
  n14 = TextureSample(UVs=Add(A=Add(A=@n2, B=@n12), B=Multiply(A=@n5, B=@n13))) {samplerType=SAMPLERTYPE_LinearColor, texture=&/VT_Voronoi_LOW}
  n15 = WorldPosition()
  n16 = ScalarParameter() {defaultValue=10, parameterName=FogMacroScale}
  n17 = Multiply(A=@n5, B=@n16)
  n18 = WorldPosition()
  n19 = Multiply(A=@n7, B=@n16)
  n20 = Multiply(A=Time(), B=VectorParameter() {parameterName=CurlNoisePanSpeedAndDirection, defaultValue=<rgba -0.02 -0.02 -0.02 1>}.RGB)
  n21 = ScalarParameter() {defaultValue=0.25, parameterName=MacroPanSpeedMultiplier}
  n22 = TextureSample(UVs=Add(A=Divide(A=@n18.XYZ, B=@n19), B=Multiply(A=@n20, B=@n21))) {samplerType=SAMPLERTYPE_LinearColor, texture=&VT_CurlNoise_LOW}
  n23 = ScalarParameter() {defaultValue=0.05, parameterName=DistortionStrength}
  n24 = Time()
  n25 = VectorParameter() {parameterName=FogNoisePanSpeed, defaultValue=<rgba 0.01 0.015 0.015 1>}
  n26 = Add(A=Add(A=Divide(A=@n15.XYZ, B=@n17), B=Multiply(A=@n22.RGB, B=@n23)), B=Multiply(A=Multiply(A=@n24, B=@n25.RGB), B=@n21))
  n27 = TextureSample(UVs=Add(A=Add(A=@n2, B=@n26), B=Multiply(A=@n17, B=@n13))) {samplerType=SAMPLERTYPE_LinearColor, texture=&VT_Voronoi_LOW}
  n28 = Distance(A=CameraPositionWS(), B=ObjectPositionWS())
  n29 = Multiply(
   A=MaterialFunctionCall() {materialFunction=/Engine/Functions/Engine_MaterialFunctions02/WorldPositionOffset/ObjectScale}.Scale X
   B=ScalarParameter() {defaultValue=2, parameterName=DistanceBlendMultiplier}
  )
  n30 = CurveAtlasRowParameter(CurveTime=Clamp(None=Divide(A=@n28, B=@n29) {constB=14000})) {
   atlas=&AthenaCurveAtlas01
   curve=&/FogOpacityCurve
   defaultValue=253
   parameterName=FogOpacityCurve
  }
  n31 = Power(Base=TextureSample(UVs=@n12) {samplerType=SAMPLERTYPE_LinearColor, texture=&VT_Voronoi_LOW}.R)
  n32 = Power(Base=TextureSample(UVs=@n26) {samplerType=SAMPLERTYPE_LinearColor, texture=&VT_Voronoi_LOW}.R)
  n33 = LinearInterpolate(A=@n31, B=@n32, Alpha=@n30.A)
  n34 = Subtract(A=Power(Base=LinearInterpolate(A=@n14.R, B=@n27.R, Alpha=@n30.A)), B=@n33)
  n35 = ScalarParameter() {defaultValue=4, parameterName=ShadowDensity}
  n36 = CurveAtlasRowParameter(CurveTime=Clamp(None=MaterialFunctionCall!(Depth=@n34, Density=@n35, Use Exp2) {materialFunction=/Engine/Functions/Engine_MaterialFunctions01/Density/ExponentialDensity}.Result)) {
   atlas=&AthenaCurveAtlas01
   curve=&/FortnitemaresFogColorCurve
   defaultValue=252
   parameterName=FogColor
  }
  n37 = Multiply(A=@n36, B=Multiply(A=@n36.A, B=ScalarParameter() {defaultValue=1, parameterName=UnderGlowStrength}))
  n38 = ObjectPositionWS()
  n39 = WorldPosition()
  n40 = Divide(A=Add(A=Subtract(A=@n38, B=@n39.XYZ), B=ObjectBounds()), B=ObjectBounds())
  n41 = CurveAtlasRowParameter(CurveTime=ComponentMask(None=Multiply(A=@n40) {constB=0.5}) {b=true}) {
   atlas=&AthenaCurveAtlas01
   curve=&FogOpacityCurve
   defaultValue=253
   parameterName=FogOpacityCurve
  }
  LinearInterpolate(A=@n36, B=Multiply(A=@n36, B=@n37), Alpha=@n41.B)
  MaterialFunctionCall!(Vector3, MaximumNumberOfDigits, UVs=ScreenPosition().ViewportUV, DebugTextLocation RG_UpperRight BA_LowerLeft, Component Spacing) {materialFunction=/Engine/Functions/Engine_MaterialFunctions02/Utility/DebugFloat3Values}
  Multiply(A=@n32, B=ScalarParameter() {defaultValue=2, parameterName=Noisetrength})
  n42 = Multiply(A=@n36, B=ScalarParameter() {defaultValue=0.25, parameterName=FogEmissiveStregnth})
  n43 = ComponentMask(None=MaterialFunctionCall(
    Color_Morning=ScalarParameter() {parameterName=Morning_FogMin}
    Color_Day=ScalarParameter() {parameterName=Day_FogMin}
    Color_Evening=ScalarParameter() {parameterName=Evening_FogMin}
    Color_Night=ScalarParameter() {parameterName=Night_FogMin}
   ) {materialFunction=&Color_Time_of_Day}.Result) {r=true}
  n44 = ScalarParameter() {parameterName=Macro_MinDensityMultiplier}
  n45 = ComponentMask(None=MaterialFunctionCall(
    Color_Morning=ScalarParameter() {defaultValue=0.05, parameterName=Morning_FogMax}
    Color_Day=ScalarParameter() {defaultValue=0.05, parameterName=Day_FogMax}
    Color_Evening=ScalarParameter() {defaultValue=0.05, parameterName=Evening_FogMax}
    Color_Night=ScalarParameter() {defaultValue=0.05, parameterName=Night_FogMax}
   ) {materialFunction=&Color_Time_of_Day}.Result) {g=true}
  n46 = ScalarParameter() {defaultValue=0.04, parameterName=Macro_MAxDensityMultiplier}
  n47 = Distance(A=WorldPosition().XYZ, B=ObjectPositionWS())
  n48 = Multiply(A=ObjectRadius(), B=ScalarParameter() {defaultValue=2, parameterName=RadiusScale}) {constB=0.9}
  n49 = StaticSwitch(
   True=OneMinus(None=Divide(A=@n47, B=@n48))
   False=Constant() {r=1}
   Value=StaticBoolParameter() {defaultValue=true, parameterName=UseRadialFalloff}
  )
  n50 = VectorParameter() {parameterName=BBoxMin}
  n51 = Subtract(A=WorldPosition().XYZ, B=@n50.RGB)
  n52 = Subtract(A=VectorParameter() {parameterName=BBoxMax}.RGB, B=@n50.RGB)
  n53 = Power(Base=Clamp(None=Divide(A=@n51, B=@n52)), Exp=VectorParameter() {parameterName=AxisShift, defaultValue=<rgba 1 1 1 0>}.RGB)
  n54 = Abs(None=Multiply(A=Subtract(A=@n53) {constB=0.5}) {constB=2})
  n55 = MaterialFunctionCall(Float3=Power(Base=OneMinus(None=@n54), Exp=VectorParameter() {parameterName=AxisFallfoffExponent, defaultValue=<rgba 1 1 1 1>}.RGB)) {materialFunction=&BreakOutFloat3Components}
  n56 = Multiply(A=@n49, B=Multiply(A=Multiply(A=@n55.R, B=@n55.R), B=@n55.B))
  n57 = Subtract(
   A=Multiply(A=StaticSwitchParameter(True=@n56, False=@n49) {defaultValue=true, parameterName=UseAxisFalloff}, B=@n41.R)
   B=Multiply(A=@n33, B=ScalarParameter() {defaultValue=2, parameterName=Noisetrength})
  )
  n58 = Clamp(None=LinearInterpolate(A=LinearInterpolate(A=@n43, B=@n44, Alpha=@n30.A), B=LinearInterpolate(A=@n45, B=@n46, Alpha=@n30.A), Alpha=Clamp(None=@n57)))
  n59 = ScalarParameter() {defaultValue=1, desc=Affects all TOD, parameterName=GlobalExtinctionMultiplier}
  Multiply(
   A=Multiply(A=LinearInterpolate(A=@n42, B=@n37, Alpha=@n36.A), B=Multiply(A=@n58, B=@n59))
   B=ScalarParameter() {defaultValue=1, parameterName=FogGlobalEmissive}
  )
  Multiply(A=@n31, B=@n32)
  ScalarParameter() {defaultValue=1024, parameterName=NoiseScale}
 ]
}
